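# Child animation Blueprint for the SnipeHover biped cosmetic companion, updating animation state every frame.
# It derives jog, run and sprint flags from horizontal owner speed, using separate on/off thresholds for hysteresis.
# Animation curves and mood flags drive emote, reaction and walking flags plus worried and antics blend alphas.
# Look-at angles pass through normally but are fixed for preview actors; a moving-based worried-alpha branch is unconnected.
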AnimBlueprint /CosmeticCompanions/Assets/Biped/SnipeHover/Blueprint/SnipeHover_AnimBP_Child
target_skeleton=/CosmeticCompanions/Assets/Biped/SnipeHover/Meshes/SnipeHover_Biped_Skeleton
graphs [EventGraph {nodes [
   Event BlueprintUpdateAnimation()
   > n1 = Sequence!(execute) -> (then_0, then_1, then_2:Exec)
   n2 = GetOwningActor()
   > Set CompanionSpeed_0(CompanionSpeed_0:double=Vector Length XY!(A:Vector=GetVelocity!(self:Actor Object Reference=@n2) -> (ReturnValue:Vector)) -> (ReturnValue:double)) -> (Output_Get:double)
   > n3 = Sequence!(execute) -> (then_0, then_1)
   > n4 = Branch(Condition=Get bIsMoodWorried() -> (bIsMoodWorried:bool))
   > Set WorriedAlpha_0(WorriedAlpha_0:double=float - float(A=1, B=GetCurveValue(CurveName=DisableWorriedAlpha))) -> (Output_Get:double)
   Set WorriedAlpha_0(execute=@n4.else, WorriedAlpha_0=0)
   n5 = Sequence!(execute=@n3.then_1) -> (then_0, then_1, then_2:Exec)
   n6 = Get CompanionSpeed_0() -> (CompanionSpeed_0:double)
   > n7 = Branch(Condition="float >= float"(A=@n6, B=675))
   > Set IsLocoSprint_0(IsLocoSprint_0:bool=true) -> (Output_Get:bool)
   Branch(execute=@n7.else, Condition="float < float"(A=@n6, B=660))
   > Set IsLocoSprint_0(IsLocoSprint_0=false)
   n8 = Branch(execute=@n5.then_1, Condition="float >= float"(A=Get CompanionSpeed_0(), B=285))
   > Set IsLocoRun_0(IsLocoRun_0:bool=true) -> (Output_Get:bool)
   Branch(execute=@n8.else, Condition="float < float"(A=Get CompanionSpeed_0(), B=270))
   > Set IsLocoRun_0(IsLocoRun_0=false)
   n9 = Branch(execute=@n5.then_2, Condition="float >= float"(A=Get CompanionSpeed_0(), B=185))
   > Set IsLocoJog_0(IsLocoJog_0:bool=true) -> (Output_Get:bool)
   Branch(execute=@n9.else, Condition="float < float"(A=Get CompanionSpeed_0(), B=175))
   > Set IsLocoJog_0(IsLocoJog_0=false)
   Set EmoteAActive(execute=@n1.then_1, EmoteAActive:bool="float > float"(A=GetCurveValue(CurveName=EmoteCurveA))) -> (Output_Get:bool)
   > Set EmoteBActive(EmoteBActive:bool="float > float"(A=GetCurveValue(CurveName=EmoteCurveB))) -> (Output_Get:bool)
   > Set HappyReactionActive_0(HappyReactionActive_0:bool="float > float"(A=GetCurveValue(CurveName=Reaction_Happy))) -> (Output_Get:bool)
   > Set IsWalking(IsWalking:bool="float > float"(A=GetCurveValue(CurveName=IsWalking), B=0.5)) -> (Output_Get:bool)
   n10 = OR Boolean!(A=Get EmoteAActive() -> (EmoteAActive:bool), B=Get EmoteBActive() -> (EmoteBActive:bool), C:bool=Get bIsMoodCombat() -> (bIsMoodCombat:bool)) -> (ReturnValue)
   > Set AnticsAlpha_0(AnticsAlpha_0:double="To Float (Boolean)"(InBool=NOT Boolean(A=@n10))) -> (Output_Get:double)
   n11 = OR Boolean!(A=Get bIsMoodBored() -> (bIsMoodBored:bool), B=Get bIsMoodWorried()) -> (ReturnValue)
   > Set BodyAnticsAlpha_0(BodyAnticsAlpha_0:double="To Float (Boolean)"(InBool=NOT Boolean(A=@n11))) -> (Output_Get:double)
   n12 = Branch(execute=@n1.then_2, Condition=Get bIsPreviewActor() -> (bIsPreviewActor:bool))
   > n13 = Branch(Condition="Equal (Integer)"(A=Get PreviewActorIndex() -> (PreviewActorIndex:int=0)))
   > Set LookAtYawAdjusted_0(LookAtYawAdjusted_0:double=-35) -> (Output_Get:double)
   > Set LookAtPitchAdjusted_0(LookAtPitchAdjusted_0:double=-5) -> (Output_Get:double)
   Set LookAtYawAdjusted_0(execute=@n13.else, LookAtYawAdjusted_0=0)
   > Set LookAtPitchAdjusted_0(LookAtPitchAdjusted_0=0)
   Set LookAtPitchAdjusted_0(execute=@n12.else, LookAtPitchAdjusted_0=Get LookAtPitch() -> (LookAtPitch:float))
   > Set LookAtYawAdjusted_0(LookAtYawAdjusted_0=Get LookAtYaw() -> (LookAtYaw:float))
   n14 = Branch(Condition=Get bIsMoving() -> (bIsMoving:bool))
   > Set WorriedAlpha_0(WorriedAlpha_0=0.5)
   Set WorriedAlpha_0(execute=@n14.else, WorriedAlpha_0=1)
   TryGetPawnOwner()
  ]}]
variables [
 AnticsAlpha_0 {type=number}
 BodyAnticsAlpha_0 {type=number}
 CompanionSpeed_0 {type=number}
 EmoteAActive {type=boolean}
 EmoteBActive {type=boolean}
 HappyReactionActive_0 {type=boolean}
 IsLocoJog_0 {type=boolean}
 IsLocoRun_0 {type=boolean}
 IsLocoSprint_0 {type=boolean}
 IsWalking {type=boolean}
 LookAtPitchAdjusted_0 {type=number}
 LookAtYawAdjusted_0 {type=number}
 WorriedAlpha_0 {type=number}
]
