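# Post-process Animation Blueprint for the SnipeHover Biped cosmetic companion skeleton that applies head facial poses.
# The AnimGraph feeds the Input Pose through a Pose Blend node using the head facial PoseAsset, which turns incoming animation curves into facial poses, then outputs the result.
# The EventGraph is empty, and a PoseAsset variable defaults to the same facial pose asset.
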AnimBlueprint /CosmeticCompanions/Assets/Biped/SnipeHover/Meshes/SnipeHover_Biped_Post_AnimBP
target_skeleton=&/SnipeHover_Biped_Skeleton
graphs [
 AnimGraph {nodes [Output Pose(Result=SnipeHover_Biped_Head_Facial_Poses_PoseAsset!(SourcePose:Pose Link Structure=Input Pose()) -> (Pose:Pose Link Structure) {
     is=title AnimGraphNode_PoseBlendNode
     poseAsset=&/SnipeHover_Biped_Head_Facial_Poses_PoseAsset
    })]}
 EventGraph {links [], nodes []}
]
variables [PoseAsset {type=PoseAsset, default=&SnipeHover_Biped_Head_Facial_Poses_PoseAsset}]
